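# Physics asset for the Vampire Hunter backpack cosmetic, providing bodies and joints for its left and right swords.
# Each side has a long, thin capsule on the sword bone and a tiny capsule on its master bone.
# A constraint links each sword to its master with tight, asymmetric swing limits and almost no twist, so any sway stays small and mostly on one axis.
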PhysicsAsset /BRCosmetics/Accessories/FORT_Backpacks/Backpack_F_MED_Vampire_Hunter/Meshes/F_MED_Vampire_Hunter_Pack_Physics
bodies {
 Dyn_MasterSword_L [{shapeType=Capsule, length=0.505, radius=0.505, shapeName=Dyn_MasterSword_L_capsule, rotation=<rot 0 -90 0>}]
 Dyn_MasterSword_R [{shapeType=Capsule, length=0.505, radius=0.505, shapeName=Dyn_MasterSword_R_capsule, rotation=<rot 0 -90 0>}]
 Dyn_Sword_L1 [{shapeType=Capsule, length=35, radius=2, center=<10 -2 0>, rotation=<rot 90 0 0>}]
 Dyn_Sword_R1 [{shapeType=Capsule, length=35, radius=2, center=<10 2 0>, rotation=<rot 90 0 0>}]
}
constraints [
 {bone1Name=Dyn_Sword_R1, bone2Name=Dyn_MasterSword_R, swing1LimitDegrees=5, swing2LimitDegrees=20, twistLimitDegrees=1}
 {bone1Name=Dyn_Sword_L1, bone2Name=Dyn_MasterSword_L, swing1LimitDegrees=5, swing2LimitDegrees=20, twistLimitDegrees=1}
]
